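# Sound cue for the NPC hover audio of the Hightower Tomato emote, classed as emote music.
# A single wave player feeds the root directly, with no randomization, looping node or mixing.
# Overridden attenuation adds distance-based low-pass filtering alongside volume falloff, and a shared emote foley concurrency set is assigned.
# A submix send also routes it to the music audio analysis submix.
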SoundCue /BRCosmetics/Sounds/Emotes/Hightower_Tomato/Emote_Hightower_Tomato_NPC_Hover
root=WavePlayer_0
nodes [{id=WavePlayer_0, properties {soundWaveAssetPtr=&/Emote_Hightower_Tomato_Hover_NPC_01}}]
properties {
 attenuationSettings=&Fort_Emotes_Athena_Att
 bOverrideAttenuation=true
 maxDistance=3250
 soundClassObject=&Fort_Music_Emote
 volumeMultiplier=1
 attenuationOverrides {bAttenuateWithLPF=true, falloffDistance=2500, lPFFrequencyAtMax=2000, lPFRadiusMax=2000, lPFRadiusMin=750, attenuationShapeExtents {x=750}}
 concurrencyOverrides {maxCount=3}
 concurrencySet [/Game/Sounds/Concurrency_Rules/Athena_EmoteFoley_Conc]
 soundSubmixSends [{minSendLevel=0, soundSubmix=&Fortnite_AudioAnalysis_Music}]
}
